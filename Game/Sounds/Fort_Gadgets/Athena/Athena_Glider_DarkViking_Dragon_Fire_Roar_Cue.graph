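# Sound cue for the fire roar of the DarkViking dragon glider in Athena.
# A single roar wave feeds a Modulator that randomizes pitch on each play for variation, at reduced cue volume.
# It uses glider thrust attenuation and the glider landing sound class.
# An inline concurrency override caps simultaneous instances and rejects new plays, taking precedence over the assigned glider open/close concurrency set.
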SoundCue /Game/Sounds/Fort_Gadgets/Athena/Athena_Glider_DarkViking_Dragon_Fire_Roar_Cue
root=Modulator_0
nodes [
 {id=Modulator_0, children [WavePlayer_1], properties {pitchMax=1.1, pitchMin=0.8}}
 {id=WavePlayer_1, properties {soundWaveAssetPtr=&/DarkViking_FireRoar_01}}
]
properties {
 attenuationSettings=&Fort_Player_Glider_Thrust_Athena_Att
 bOverrideConcurrency=true
 maxDistance=12500
 soundClassObject=&Fort_Gliders_Land
 volumeMultiplier=0.35
 concurrencyOverrides {maxCount=2, resolutionRule=PreventNew}
 concurrencySet [&Athena_GliderOpenClose_Conc]
}
