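# Sound cue for the dragon grunts of the DarkViking dragon glider in Athena.
# A random node picks one of four equally weighted grunt waves, then a modulator randomizes its pitch around normal.
# Playback is quiet, using glider thrust attenuation and the glider landing sound class.
# Overridden concurrency allows one instance and blocks new ones, alongside a referenced glider open/close concurrency set.
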SoundCue /Game/Sounds/Fort_Gadgets/Athena/Athena_Glider_DarkViking_Dragon_Grunts_Cue
root=Modulator_0
nodes [
 {id=Modulator_0, children [Random_0], properties {pitchMax=1.1, pitchMin=0.8}}
 {id=Random_0, children [WavePlayer_0,WavePlayer_2,WavePlayer_3,WavePlayer_4], properties {weights [1,1,1,1]}}
 {id=WavePlayer_0, properties {soundWaveAssetPtr=&/DarkViking_Dragon_Grunts_11}}
 {id=WavePlayer_2, properties {soundWaveAssetPtr=&/DarkViking_Dragon_Grunts_12}}
 {id=WavePlayer_3, properties {soundWaveAssetPtr=&/DarkViking_Dragon_Grunts_13}}
 {id=WavePlayer_4, properties {soundWaveAssetPtr=&/DarkViking_Dragon_Grunts_14}}
]
properties {
 attenuationSettings=&Fort_Player_Glider_Thrust_Athena_Att
 bOverrideConcurrency=true
 maxDistance=12500
 soundClassObject=&Fort_Gliders_Land
 volumeMultiplier=0.2
 concurrencyOverrides {maxCount=1, resolutionRule=PreventNew}
 concurrencySet [&Athena_GliderOpenClose_Conc]
}
