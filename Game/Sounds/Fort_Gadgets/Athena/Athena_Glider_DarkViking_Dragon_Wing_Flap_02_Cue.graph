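# Sound cue for a wing flap of the DarkViking Dragon glider in Athena, playing the WingFlap_03 wave despite the cue's 02 name.
# A single wave player feeds a modulator that randomizes pitch on each play, with the cue's overall volume lowered.
# It uses glider thrust attenuation and the glider landing sound class.
# Overridden concurrency caps simultaneous instances per owner and stops the oldest, though a glider open/close concurrency set is also assigned.
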SoundCue /Game/Sounds/Fort_Gadgets/Athena/Athena_Glider_DarkViking_Dragon_Wing_Flap_02_Cue
root=Modulator_0
nodes [
 {id=Modulator_0, children [WavePlayer_0], properties {pitchMax=1.25, pitchMin=0.85, volumeMin=1}}
 {id=WavePlayer_0, properties {soundWaveAssetPtr=&/DarkViking_Dragon_WingFlap_03}}
]
properties {
 attenuationSettings=&Fort_Player_Glider_Thrust_Athena_Att
 bOverrideConcurrency=true
 maxDistance=12500
 soundClassObject=&Fort_Gliders_Land
 volumeMultiplier=0.4
 concurrencyOverrides {bLimitToOwner=true, maxCount=2, resolutionRule=StopOldest}
 concurrencySet [&Athena_GliderOpenClose_Conc]
}
