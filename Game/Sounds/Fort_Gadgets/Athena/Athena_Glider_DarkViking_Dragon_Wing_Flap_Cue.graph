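# Sound cue for the wing flap of the Dark Viking dragon glider in Athena.
# A random node picks one of two equally weighted flap waves, and a modulator varies its pitch on each play.
# It uses glider thrust attenuation and a glider sound class.
# An owner-limited concurrency override caps overlapping flaps by stopping the oldest, alongside a referenced glider open/close concurrency set.
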SoundCue /Game/Sounds/Fort_Gadgets/Athena/Athena_Glider_DarkViking_Dragon_Wing_Flap_Cue
root=Modulator_0
nodes [
 {id=Modulator_0, children [Random_0], properties {pitchMax=1.1, pitchMin=0.85, volumeMin=1}}
 {id=Random_0, children [WavePlayer_1,WavePlayer_2], properties {weights [1,1]}}
 {id=WavePlayer_1, properties {soundWaveAssetPtr=&/DarkViking_Dragon_WingFlap_01}}
 {id=WavePlayer_2, properties {soundWaveAssetPtr=&/DarkViking_Dragon_WingFlap_02}}
]
properties {
 attenuationSettings=&Fort_Player_Glider_Thrust_Athena_Att
 bOverrideConcurrency=true
 maxDistance=12500
 soundClassObject=&Fort_Gliders_Land
 volumeMultiplier=0.4
 concurrencyOverrides {bLimitToOwner=true, maxCount=2, resolutionRule=StopOldest}
 concurrencySet [&Athena_GliderOpenClose_Conc]
}
