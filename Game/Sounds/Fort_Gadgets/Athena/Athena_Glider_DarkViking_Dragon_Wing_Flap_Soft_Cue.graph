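# Sound cue for the soft wing flap of the DarkViking dragon glider in Athena.
# It randomly picks one of two equally weighted flap waves and runs it through a modulator for slight pitch variation.
# It plays at reduced volume in the glider landing sound class with glider thrust attenuation.
# A concurrency override caps simultaneous instances at three, taking precedence over the listed glider open/close concurrency set.
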SoundCue /Game/Sounds/Fort_Gadgets/Athena/Athena_Glider_DarkViking_Dragon_Wing_Flap_Soft_Cue
root=Modulator_0
nodes [
 {id=Modulator_0, children [Random_1], properties {pitchMax=1.1, pitchMin=0.9}}
 {id=Random_1, children [WavePlayer_0,WavePlayer_3], properties {weights [1,1]}}
 {id=WavePlayer_0, properties {soundWaveAssetPtr=&/DarkViking_Dragon_WingFlap_Soft_01}}
 {id=WavePlayer_3, properties {soundWaveAssetPtr=&/DarkViking_Dragon_WingFlap_Soft_02}}
]
properties {
 attenuationSettings=&Fort_Player_Glider_Thrust_Athena_Att
 bOverrideConcurrency=true
 maxDistance=12500
 soundClassObject=&Fort_Gliders_Land
 volumeMultiplier=0.25
 concurrencyOverrides {maxCount=3}
 concurrencySet [&Athena_GliderOpenClose_Conc]
}
