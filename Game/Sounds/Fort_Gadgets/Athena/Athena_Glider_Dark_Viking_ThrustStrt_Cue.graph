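# Sound cue for the thrust-start sound of the Athena Dark Viking glider, with no sound nodes or root, so it holds no audio to play.
# It still carries playback settings: a glider thrust attenuation asset, the glider landing sound class and a lowered volume.
# It references a shared glider thrust concurrency set and also holds single-instance, prevent-new concurrency override values.
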SoundCue /Game/Sounds/Fort_Gadgets/Athena/Athena_Glider_Dark_Viking_ThrustStrt_Cue
root=null
nodes []
properties {
 attenuationSettings=&Fort_Player_Glider_Thrust_Athena_Att
 maxDistance=10000
 soundClassObject=&Fort_Gliders_Land
 volumeMultiplier=0.15
 concurrencyOverrides {maxCount=1, resolutionRule=PreventNew}
 concurrencySet [&Athena_Glider_Thrust_Conc]
}
